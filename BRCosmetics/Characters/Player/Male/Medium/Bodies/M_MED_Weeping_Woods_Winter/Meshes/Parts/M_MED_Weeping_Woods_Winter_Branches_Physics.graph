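# Physics asset for the branches part of the Weeping Woods Winter male medium body cosmetic, defining bodies for its dynamic branch bones.
# Left, right and central branch chains hang from tiny master anchor capsules constrained to the head, with the left chain forking in two.
# Branch joints have tight swing and twist limits, while the anchor, head and spine joints leave rotation free.
# Larger capsules cover the head and upper spine.
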PhysicsAsset /BRCosmetics/Characters/Player/Male/Medium/Bodies/M_MED_Weeping_Woods_Winter/Meshes/Parts/M_MED_Weeping_Woods_Winter_Branches_Physics
bodies {
 dyn_branch_01_MSTR [{shapeType=Capsule, length=0.01, radius=0.01}]
 dyn_branch_01_l [{shapeType=Capsule, length=6.098, radius=1.25, center=<2.6888 0 0>, rotation=<rot 90 0.0791 0.0791>}]
 dyn_branch_01_l_MSTR [{shapeType=Capsule, length=0.01, radius=0.01}]
 dyn_branch_01_r [{shapeType=Capsule, length=6.098, radius=1.25, center=<2.6888 0 0>, rotation=<rot 90 0.0791 0.0791>}]
 dyn_branch_01_r_MSTR [{shapeType=Capsule, length=0.01, radius=0.01}]
 dyn_branch_02_l [{shapeType=Box, extentX=4.5, extentY=2.5, extentZ=0.5, center=<2.7237 0 0>}]
 dyn_branch_02_r [{shapeType=Box, extentX=4.5, extentY=2.5, extentZ=0.5, center=<2.577 0 0>}]
 dyn_branch_03_l [{shapeType=Box, extentX=4.5, extentY=2.5, extentZ=0.5, center=<2.577 0 0>}]
 dyn_branch_05 [{shapeType=Box, extentX=2, extentY=2, extentZ=2, center=<1 0 0>}]
 head [{shapeType=Capsule, length=14.7278, radius=8.0113, rotation=<rot -90 -1.2659 1.2659>}]
 spine_04 [{shapeType=Capsule, length=7.02, radius=14.2494, center=<4.243 -3.3415 0.8424>, rotation=<rot 0 90 0>}]
 spine_05 [{shapeType=Capsule, length=9.77, radius=11.4994, center=<4.243 -3.3415 0.8424>, rotation=<rot 0 90 0>}]
}
constraints [
 {bone1Name=dyn_branch_01_r, bone2Name=dyn_branch_01_r_MSTR, swing1LimitDegrees=10, swing2LimitDegrees=10, twistLimitDegrees=5}
 {bone1Name=dyn_branch_03_l, bone2Name=dyn_branch_01_l, swing1LimitDegrees=10, swing2LimitDegrees=10, twistLimitDegrees=5}
 {bone1Name=dyn_branch_02_l, bone2Name=dyn_branch_01_l, swing1LimitDegrees=10, swing2LimitDegrees=10, twistLimitDegrees=5}
 {bone1Name=dyn_branch_02_r, bone2Name=dyn_branch_01_r, swing1LimitDegrees=10, swing2LimitDegrees=10, twistLimitDegrees=5}
 {bone1Name=dyn_branch_05, bone2Name=dyn_branch_01_MSTR, swing1LimitDegrees=10, swing2LimitDegrees=10, twistLimitDegrees=5}
 {bone1Name=dyn_branch_01_l, bone2Name=dyn_branch_01_l_MSTR, swing1LimitDegrees=10, swing2LimitDegrees=10, twistLimitDegrees=5}
 {
  bone1Name=dyn_branch_01_r_MSTR
  bone2Name=head
  swing1LimitDegrees=45
  swing1Motion=Free
  swing2LimitDegrees=45
  swing2Motion=Free
  twistLimitDegrees=45
  twistMotion=Free
 }
 {
  bone1Name=dyn_branch_01_MSTR
  bone2Name=head
  swing1LimitDegrees=45
  swing1Motion=Free
  swing2LimitDegrees=45
  swing2Motion=Free
  twistLimitDegrees=45
  twistMotion=Free
 }
 {
  bone1Name=dyn_branch_01_l_MSTR
  bone2Name=head
  swing1LimitDegrees=45
  swing1Motion=Free
  swing2LimitDegrees=45
  swing2Motion=Free
  twistLimitDegrees=45
  twistMotion=Free
 }
 {
  bone1Name=head
  bone2Name=spine_05
  swing1LimitDegrees=45
  swing1Motion=Free
  swing2LimitDegrees=45
  swing2Motion=Free
  twistLimitDegrees=45
  twistMotion=Free
 }
 {
  bone1Name=spine_05
  bone2Name=spine_04
  swing1LimitDegrees=45
  swing1Motion=Free
  swing2LimitDegrees=45
  swing2Motion=Free
  twistLimitDegrees=45
  twistMotion=Free
 }
]
