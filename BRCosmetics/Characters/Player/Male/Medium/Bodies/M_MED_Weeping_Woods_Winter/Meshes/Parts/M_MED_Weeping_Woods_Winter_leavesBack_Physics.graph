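# Physics asset for the back-leaves mesh part of the Weeping Woods Winter player cosmetic, defining bodies for leaf, branch, neck and shoulder bones around a head capsule.
# Each numbered head leaf is a box constrained to its own tiny MSTR anchor capsule, swinging mainly on one axis.
# Those anchors, plus other headleaf and branch MSTR capsules, constrain to the head with free angular motion.
# Neck and shoulder anchors have no constraints.
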PhysicsAsset /BRCosmetics/Characters/Player/Male/Medium/Bodies/M_MED_Weeping_Woods_Winter/Meshes/Parts/M_MED_Weeping_Woods_Winter_leavesBack_Physics
bodies {
 dyn_branch_01_MSTR [{shapeType=Capsule, length=0.01, radius=0.01}]
 dyn_branch_01_l_MSTR [{shapeType=Capsule, length=0.01, radius=0.01}]
 dyn_branch_01_r_MSTR [{shapeType=Capsule, length=0.01, radius=0.01}]
 dyn_head_leaf_01 [{shapeType=Box, extentX=5.5, extentY=4, extentZ=2, center=<2.5614 0 0>}]
 dyn_head_leaf_012 [{shapeType=Box, extentX=5.5, extentY=4, extentZ=2, center=<2.5614 0 0>}]
 dyn_head_leaf_012_MSTR [{shapeType=Capsule, length=0.01, radius=0.01}]
 dyn_head_leaf_013 [{shapeType=Box, extentX=5.5, extentY=4, extentZ=2, center=<2.5614 0 0>}]
 dyn_head_leaf_013_MSTR [{shapeType=Capsule, length=0.01, radius=0.01}]
 dyn_head_leaf_014 [{shapeType=Box, extentX=5.5, extentY=4, extentZ=2, center=<2.5614 0 0>}]
 dyn_head_leaf_014_MSTR [{shapeType=Capsule, length=0.01, radius=0.01}]
 dyn_head_leaf_015 [{shapeType=Box, extentX=5.5, extentY=4, extentZ=2, center=<2.5614 0 0>}]
 dyn_head_leaf_015_MSTR [{shapeType=Capsule, length=0.01, radius=0.01}]
 dyn_head_leaf_016 [{shapeType=Box, extentX=5.5, extentY=4, extentZ=2, center=<2.5614 0 0>}]
 dyn_head_leaf_016_MSTR [{shapeType=Capsule, length=0.01, radius=0.01}]
 dyn_head_leaf_017 [{shapeType=Box, extentX=5.5, extentY=4, extentZ=2, center=<2.5614 0 0>}]
 dyn_head_leaf_017_MSTR [{shapeType=Capsule, length=0.01, radius=0.01}]
 dyn_head_leaf_018 [{shapeType=Box, extentX=5.5, extentY=4, extentZ=2, center=<2.5614 0 0>}]
 dyn_head_leaf_018_MSTR [{shapeType=Capsule, length=0.01, radius=0.01}]
 dyn_head_leaf_019 [{shapeType=Box, extentX=5.5, extentY=4, extentZ=2, center=<2.5614 0 0>}]
 dyn_head_leaf_019_MSTR [{shapeType=Capsule, length=0.01, radius=0.01}]
 dyn_head_leaf_01_MSTR [{shapeType=Capsule, length=0.01, radius=0.01}]
 dyn_head_leaf_02 [{shapeType=Box, extentX=5.5, extentY=4, extentZ=2, center=<2.5614 0 0>}]
 dyn_head_leaf_020 [{shapeType=Box, extentX=5.5, extentY=4, extentZ=2, center=<2.5614 0 0>}]
 dyn_head_leaf_020_MSTR [{shapeType=Capsule, length=0.01, radius=0.01}]
 dyn_head_leaf_023 [{shapeType=Box, extentX=5.5, extentY=4, extentZ=2, center=<2.5614 0 0>}]
 dyn_head_leaf_023_MSTR [{shapeType=Capsule, length=0.01, radius=0.01}]
 dyn_head_leaf_026 [{shapeType=Box, extentX=5.5, extentY=4, extentZ=2, center=<2.5614 0 0>}]
 dyn_head_leaf_026_MSTR [{shapeType=Capsule, length=0.01, radius=0.01}]
 dyn_head_leaf_027 [{shapeType=Box, extentX=5.5, extentY=4, extentZ=2, center=<2.5614 0 0>}]
 dyn_head_leaf_027_MSTR [{shapeType=Capsule, length=0.01, radius=0.01}]
 dyn_head_leaf_028 [{shapeType=Box, extentX=5.5, extentY=4, extentZ=2, center=<2.5614 0 0>}]
 dyn_head_leaf_028_MSTR [{shapeType=Capsule, length=0.01, radius=0.01}]
 dyn_head_leaf_029 [{shapeType=Box, extentX=5.5, extentY=4, extentZ=2, center=<2.5614 0 0>}]
 dyn_head_leaf_029_MSTR [{shapeType=Capsule, length=0.01, radius=0.01}]
 dyn_head_leaf_02_MSTR [{shapeType=Capsule, length=0.01, radius=0.01}]
 dyn_head_leaf_03 [{shapeType=Box, extentX=5.5, extentY=4, extentZ=2, center=<2.5614 0 0>}]
 dyn_head_leaf_030 [{shapeType=Box, extentX=5.5, extentY=4, extentZ=2, center=<2.5614 0 0>}]
 dyn_head_leaf_030_MSTR [{shapeType=Capsule, length=0.01, radius=0.01}]
 dyn_head_leaf_032 [{shapeType=Box, extentX=5.5, extentY=4, extentZ=2, center=<2.5614 0 0>}]
 dyn_head_leaf_032_MSTR [{shapeType=Capsule, length=0.01, radius=0.01}]
 dyn_head_leaf_033 [{shapeType=Box, extentX=5.5, extentY=4, extentZ=2, center=<2.5614 0 0>}]
 dyn_head_leaf_033_MSTR [{shapeType=Capsule, length=0.01, radius=0.01}]
 dyn_head_leaf_034 [{shapeType=Box, extentX=5.5, extentY=4, extentZ=2, center=<2.5614 0 0>}]
 dyn_head_leaf_034_MSTR [{shapeType=Capsule, length=0.01, radius=0.01}]
 dyn_head_leaf_035 [{shapeType=Box, extentX=5.5, extentY=4, extentZ=2, center=<2.5614 0 0>}]
 dyn_head_leaf_035_MSTR [{shapeType=Capsule, length=0.01, radius=0.01}]
 dyn_head_leaf_036 [{shapeType=Box, extentX=5.5, extentY=4, extentZ=2, center=<2.5614 0 0>}]
 dyn_head_leaf_036_MSTR [{shapeType=Capsule, length=0.01, radius=0.01}]
 dyn_head_leaf_037 [{shapeType=Box, extentX=5.5, extentY=4, extentZ=2, center=<2.5614 0 0>}]
 dyn_head_leaf_037_MSTR [{shapeType=Capsule, length=0.01, radius=0.01}]
 dyn_head_leaf_039 [{shapeType=Box, extentX=5.5, extentY=4, extentZ=2, center=<2.5614 0 0>}]
 dyn_head_leaf_039_MSTR [{shapeType=Capsule, length=0.01, radius=0.01}]
 dyn_head_leaf_03_MSTR [{shapeType=Capsule, length=0.01, radius=0.01}]
 dyn_head_leaf_040 [{shapeType=Box, extentX=5.5, extentY=4, extentZ=2, center=<2.5614 0 0>}]
 dyn_head_leaf_040_MSTR [{shapeType=Capsule, length=0.01, radius=0.01}]
 dyn_head_leaf_041 [{shapeType=Box, extentX=5.5, extentY=4, extentZ=2, center=<2.5614 0 0>}]
 dyn_head_leaf_041_MSTR [{shapeType=Capsule, length=0.01, radius=0.01}]
 dyn_head_leaf_042 [{shapeType=Box, extentX=5.5, extentY=4, extentZ=2, center=<2.5614 0 0>}]
 dyn_head_leaf_042_MSTR [{shapeType=Capsule, length=0.01, radius=0.01}]
 dyn_head_leaf_05 [{shapeType=Box, extentX=5.5, extentY=4, extentZ=2, center=<2.5614 0 0>}]
 dyn_head_leaf_05_MSTR [{shapeType=Capsule, length=0.01, radius=0.01}]
 dyn_head_leaf_06 [{shapeType=Box, extentX=5.5, extentY=4, extentZ=2, center=<2.5614 0 0>}]
 dyn_head_leaf_06_MSTR [{shapeType=Capsule, length=0.01, radius=0.01}]
 dyn_head_leaf_07 [{shapeType=Box, extentX=5.5, extentY=4, extentZ=2, center=<2.5614 0 0>}]
 dyn_head_leaf_07_MSTR [{shapeType=Capsule, length=0.01, radius=0.01}]
 dyn_head_leaf_08 [{shapeType=Box, extentX=5.5, extentY=4, extentZ=2, center=<2.5614 0 0>}]
 dyn_head_leaf_08_MSTR [{shapeType=Capsule, length=0.01, radius=0.01}]
 dyn_head_leaf_09 [{shapeType=Box, extentX=5.5, extentY=4, extentZ=2, center=<2.5614 0 0>}]
 dyn_head_leaf_09_MSTR [{shapeType=Capsule, length=0.01, radius=0.01}]
 dyn_headleaf10_01_l_MSTR [{shapeType=Capsule, length=0.01, radius=0.01}]
 dyn_headleaf10_01_r_MSTR [{shapeType=Capsule, length=0.01, radius=0.01}]
 dyn_headleaf11_01_l_MSTR [{shapeType=Capsule, length=0.01, radius=0.01}]
 dyn_headleaf1_01_l_MSTR [{shapeType=Capsule, length=0.01, radius=0.01}]
 dyn_headleaf2_01_l_MSTR [{shapeType=Capsule, length=0.01, radius=0.01}]
 dyn_headleaf2_01_r_MSTR [{shapeType=Capsule, length=0.01, radius=0.01}]
 dyn_headleaf3_01_l_MSTR [{shapeType=Capsule, length=0.01, radius=0.01}]
 dyn_headleaf3_01_r_MSTR [{shapeType=Capsule, length=0.01, radius=0.01}]
 dyn_headleaf4_01_l_MSTR [{shapeType=Capsule, length=0.01, radius=0.01}]
 dyn_headleaf4_01_r_MSTR [{shapeType=Capsule, length=0.01, radius=0.01}]
 dyn_headleaf5_01_l_MSTR [{shapeType=Capsule, length=0.01, radius=0.01}]
 dyn_headleaf5_01_r_MSTR [{shapeType=Capsule, length=0.01, radius=0.01}]
 dyn_headleaf6_01_l_MSTR [{shapeType=Capsule, length=0.01, radius=0.01}]
 dyn_headleaf7_01_l_MSTR [{shapeType=Capsule, length=0.01, radius=0.01}]
 dyn_headleaf7_01_r_MSTR [{shapeType=Capsule, length=0.01, radius=0.01}]
 dyn_headleaf8_01_l_MSTR [{shapeType=Capsule, length=0.01, radius=0.01}]
 dyn_headleaf8_01_r_MSTR [{shapeType=Capsule, length=0.01, radius=0.01}]
 dyn_headleaf9_01_l_MSTR [{shapeType=Capsule, length=0.01, radius=0.01}]
 dyn_headleaf9_01_r_MSTR [{shapeType=Capsule, length=0.01, radius=0.01}]
 dyn_neck_bk_01_MSTR [{shapeType=Capsule, length=0.01, radius=0.01}]
 dyn_neck_bk_01_l_MSTR [{shapeType=Capsule, length=0.01, radius=0.01}]
 dyn_neck_bk_01_r_MSTR [{shapeType=Capsule, length=0.01, radius=0.01}]
 dyn_neck_fr_01_MSTR [{shapeType=Capsule, length=0.01, radius=0.01}]
 dyn_neck_fr_01_l_MSTR [{shapeType=Capsule, length=0.01, radius=0.01}]
 dyn_neck_fr_01_r_MSTR [{shapeType=Capsule, length=0.01, radius=0.01}]
 dyn_shoulder_01_l_MSTR [{shapeType=Capsule, length=0.01, radius=0.01}]
 dyn_shoulder_01_r_MSTR [{shapeType=Capsule, length=0.01, radius=0.01}]
 head [{shapeType=Capsule, length=14.7278, radius=8.0113, rotation=<rot -90 -1.2659 1.2659>}]
}
constraints [
 {bone1Name=dyn_head_leaf_030, bone2Name=dyn_head_leaf_030_MSTR, swing1LimitDegrees=5, swing2LimitDegrees=45, twistLimitDegrees=5}
 {bone1Name=dyn_head_leaf_029, bone2Name=dyn_head_leaf_029_MSTR, swing1LimitDegrees=5, swing2LimitDegrees=45, twistLimitDegrees=5}
 {bone1Name=dyn_head_leaf_026, bone2Name=dyn_head_leaf_026_MSTR, swing1LimitDegrees=5, swing2LimitDegrees=45, twistLimitDegrees=5}
 {bone1Name=dyn_head_leaf_036, bone2Name=dyn_head_leaf_036_MSTR, swing1LimitDegrees=5, swing2LimitDegrees=45, twistLimitDegrees=5}
 {bone1Name=dyn_head_leaf_037, bone2Name=dyn_head_leaf_037_MSTR, swing1LimitDegrees=5, swing2LimitDegrees=45, twistLimitDegrees=5}
 {bone1Name=dyn_head_leaf_039, bone2Name=dyn_head_leaf_039_MSTR, swing1LimitDegrees=5, swing2LimitDegrees=45, twistLimitDegrees=5}
 {bone1Name=dyn_head_leaf_040, bone2Name=dyn_head_leaf_040_MSTR, swing1LimitDegrees=5, swing2LimitDegrees=45, twistLimitDegrees=5}
 {bone1Name=dyn_head_leaf_042, bone2Name=dyn_head_leaf_042_MSTR, swing1LimitDegrees=5, swing2LimitDegrees=45, twistLimitDegrees=5}
 {bone1Name=dyn_head_leaf_041, bone2Name=dyn_head_leaf_041_MSTR, swing1LimitDegrees=5, swing2LimitDegrees=45, twistLimitDegrees=5}
 {bone1Name=dyn_head_leaf_013, bone2Name=dyn_head_leaf_013_MSTR, swing1LimitDegrees=5, swing2LimitDegrees=45, twistLimitDegrees=5}
 {bone1Name=dyn_head_leaf_014, bone2Name=dyn_head_leaf_014_MSTR, swing1LimitDegrees=5, swing2LimitDegrees=45, twistLimitDegrees=5}
 {bone1Name=dyn_head_leaf_015, bone2Name=dyn_head_leaf_015_MSTR, swing1LimitDegrees=5, swing2LimitDegrees=45, twistLimitDegrees=5}
 {bone1Name=dyn_head_leaf_016, bone2Name=dyn_head_leaf_016_MSTR, swing1LimitDegrees=5, swing2LimitDegrees=45, twistLimitDegrees=5}
 {bone1Name=dyn_head_leaf_017, bone2Name=dyn_head_leaf_017_MSTR, swing1LimitDegrees=5, swing2LimitDegrees=45, twistLimitDegrees=5}
 {bone1Name=dyn_head_leaf_018, bone2Name=dyn_head_leaf_018_MSTR, swing1LimitDegrees=5, swing2LimitDegrees=45, twistLimitDegrees=5}
 {bone1Name=dyn_head_leaf_019, bone2Name=dyn_head_leaf_019_MSTR, swing1LimitDegrees=5, swing2LimitDegrees=45, twistLimitDegrees=5}
 {bone1Name=dyn_head_leaf_020, bone2Name=dyn_head_leaf_020_MSTR, swing1LimitDegrees=5, swing2LimitDegrees=45, twistLimitDegrees=5}
 {bone1Name=dyn_head_leaf_023, bone2Name=dyn_head_leaf_023_MSTR, swing1LimitDegrees=5, swing2LimitDegrees=45, twistLimitDegrees=5}
 {bone1Name=dyn_head_leaf_032, bone2Name=dyn_head_leaf_032_MSTR, swing1LimitDegrees=5, swing2LimitDegrees=45, twistLimitDegrees=5}
 {bone1Name=dyn_head_leaf_01, bone2Name=dyn_head_leaf_01_MSTR, swing1LimitDegrees=5, swing2LimitDegrees=45, twistLimitDegrees=5}
 {bone1Name=dyn_head_leaf_02, bone2Name=dyn_head_leaf_02_MSTR, swing1LimitDegrees=5, swing2LimitDegrees=45, twistLimitDegrees=5}
 {bone1Name=dyn_head_leaf_03, bone2Name=dyn_head_leaf_03_MSTR, swing1LimitDegrees=5, swing2LimitDegrees=45, twistLimitDegrees=5}
 {bone1Name=dyn_head_leaf_05, bone2Name=dyn_head_leaf_05_MSTR, swing1LimitDegrees=5, swing2LimitDegrees=45, twistLimitDegrees=5}
 {bone1Name=dyn_head_leaf_07, bone2Name=dyn_head_leaf_07_MSTR, swing1LimitDegrees=5, swing2LimitDegrees=45, twistLimitDegrees=5}
 {bone1Name=dyn_head_leaf_08, bone2Name=dyn_head_leaf_08_MSTR, swing1LimitDegrees=5, swing2LimitDegrees=45, twistLimitDegrees=5}
 {bone1Name=dyn_head_leaf_09, bone2Name=dyn_head_leaf_09_MSTR, swing1LimitDegrees=5, swing2LimitDegrees=45, twistLimitDegrees=5}
 {bone1Name=dyn_head_leaf_012, bone2Name=dyn_head_leaf_012_MSTR, swing1LimitDegrees=5, swing2LimitDegrees=45, twistLimitDegrees=5}
 {bone1Name=dyn_head_leaf_033, bone2Name=dyn_head_leaf_033_MSTR, swing1LimitDegrees=5, swing2LimitDegrees=45, twistLimitDegrees=5}
 {bone1Name=dyn_head_leaf_034, bone2Name=dyn_head_leaf_034_MSTR, swing1LimitDegrees=5, swing2LimitDegrees=45, twistLimitDegrees=5}
 {bone1Name=dyn_head_leaf_035, bone2Name=dyn_head_leaf_035_MSTR, swing1LimitDegrees=5, swing2LimitDegrees=45, twistLimitDegrees=5}
 {bone1Name=dyn_head_leaf_027, bone2Name=dyn_head_leaf_027_MSTR, swing1LimitDegrees=5, swing2LimitDegrees=45, twistLimitDegrees=5}
 {bone1Name=dyn_head_leaf_028, bone2Name=dyn_head_leaf_028_MSTR, swing1LimitDegrees=5, swing2LimitDegrees=45, twistLimitDegrees=5}
 {bone1Name=dyn_head_leaf_06, bone2Name=dyn_head_leaf_06_MSTR, swing1LimitDegrees=5, swing2LimitDegrees=45, twistLimitDegrees=5}
 {
  bone1Name=dyn_headleaf10_01_r_MSTR
  bone2Name=head
  swing1LimitDegrees=45
  swing1Motion=Free
  swing2LimitDegrees=45
  swing2Motion=Free
  twistLimitDegrees=45
  twistMotion=Free
 }
 {
  bone1Name=dyn_headleaf9_01_r_MSTR
  bone2Name=head
  swing1LimitDegrees=45
  swing1Motion=Free
  swing2LimitDegrees=45
  swing2Motion=Free
  twistLimitDegrees=45
  twistMotion=Free
 }
 {
  bone1Name=dyn_headleaf8_01_r_MSTR
  bone2Name=head
  swing1LimitDegrees=45
  swing1Motion=Free
  swing2LimitDegrees=45
  swing2Motion=Free
  twistLimitDegrees=45
  twistMotion=Free
 }
 {
  bone1Name=dyn_headleaf7_01_r_MSTR
  bone2Name=head
  swing1LimitDegrees=45
  swing1Motion=Free
  swing2LimitDegrees=45
  swing2Motion=Free
  twistLimitDegrees=45
  twistMotion=Free
 }
 {
  bone1Name=dyn_headleaf5_01_r_MSTR
  bone2Name=head
  swing1LimitDegrees=45
  swing1Motion=Free
  swing2LimitDegrees=45
  swing2Motion=Free
  twistLimitDegrees=45
  twistMotion=Free
 }
 {
  bone1Name=dyn_headleaf4_01_r_MSTR
  bone2Name=head
  swing1LimitDegrees=45
  swing1Motion=Free
  swing2LimitDegrees=45
  swing2Motion=Free
  twistLimitDegrees=45
  twistMotion=Free
 }
 {
  bone1Name=dyn_headleaf3_01_r_MSTR
  bone2Name=head
  swing1LimitDegrees=45
  swing1Motion=Free
  swing2LimitDegrees=45
  swing2Motion=Free
  twistLimitDegrees=45
  twistMotion=Free
 }
 {
  bone1Name=dyn_headleaf2_01_r_MSTR
  bone2Name=head
  swing1LimitDegrees=45
  swing1Motion=Free
  swing2LimitDegrees=45
  swing2Motion=Free
  twistLimitDegrees=45
  twistMotion=Free
 }
 {
  bone1Name=dyn_headleaf4_01_l_MSTR
  bone2Name=head
  swing1LimitDegrees=45
  swing1Motion=Free
  swing2LimitDegrees=45
  swing2Motion=Free
  twistLimitDegrees=45
  twistMotion=Free
 }
 {
  bone1Name=dyn_headleaf3_01_l_MSTR
  bone2Name=head
  swing1LimitDegrees=45
  swing1Motion=Free
  swing2LimitDegrees=45
  swing2Motion=Free
  twistLimitDegrees=45
  twistMotion=Free
 }
 {
  bone1Name=dyn_headleaf1_01_l_MSTR
  bone2Name=head
  swing1LimitDegrees=45
  swing1Motion=Free
  swing2LimitDegrees=45
  swing2Motion=Free
  twistLimitDegrees=45
  twistMotion=Free
 }
 {
  bone1Name=dyn_headleaf2_01_l_MSTR
  bone2Name=head
  swing1LimitDegrees=45
  swing1Motion=Free
  swing2LimitDegrees=45
  swing2Motion=Free
  twistLimitDegrees=45
  twistMotion=Free
 }
 {
  bone1Name=dyn_headleaf5_01_l_MSTR
  bone2Name=head
  swing1LimitDegrees=45
  swing1Motion=Free
  swing2LimitDegrees=45
  swing2Motion=Free
  twistLimitDegrees=45
  twistMotion=Free
 }
 {
  bone1Name=dyn_headleaf6_01_l_MSTR
  bone2Name=head
  swing1LimitDegrees=45
  swing1Motion=Free
  swing2LimitDegrees=45
  swing2Motion=Free
  twistLimitDegrees=45
  twistMotion=Free
 }
 {
  bone1Name=dyn_headleaf10_01_l_MSTR
  bone2Name=head
  swing1LimitDegrees=45
  swing1Motion=Free
  swing2LimitDegrees=45
  swing2Motion=Free
  twistLimitDegrees=45
  twistMotion=Free
 }
 {
  bone1Name=dyn_headleaf11_01_l_MSTR
  bone2Name=head
  swing1LimitDegrees=45
  swing1Motion=Free
  swing2LimitDegrees=45
  swing2Motion=Free
  twistLimitDegrees=45
  twistMotion=Free
 }
 {
  bone1Name=dyn_headleaf9_01_l_MSTR
  bone2Name=head
  swing1LimitDegrees=45
  swing1Motion=Free
  swing2LimitDegrees=45
  swing2Motion=Free
  twistLimitDegrees=45
  twistMotion=Free
 }
 {
  bone1Name=dyn_headleaf8_01_l_MSTR
  bone2Name=head
  swing1LimitDegrees=45
  swing1Motion=Free
  swing2LimitDegrees=45
  swing2Motion=Free
  twistLimitDegrees=45
  twistMotion=Free
 }
 {
  bone1Name=dyn_headleaf7_01_l_MSTR
  bone2Name=head
  swing1LimitDegrees=45
  swing1Motion=Free
  swing2LimitDegrees=45
  swing2Motion=Free
  twistLimitDegrees=45
  twistMotion=Free
 }
 {
  bone1Name=dyn_branch_01_r_MSTR
  bone2Name=head
  swing1LimitDegrees=45
  swing1Motion=Free
  swing2LimitDegrees=45
  swing2Motion=Free
  twistLimitDegrees=45
  twistMotion=Free
 }
 {
  bone1Name=dyn_branch_01_MSTR
  bone2Name=head
  swing1LimitDegrees=45
  swing1Motion=Free
  swing2LimitDegrees=45
  swing2Motion=Free
  twistLimitDegrees=45
  twistMotion=Free
 }
 {
  bone1Name=dyn_branch_01_l_MSTR
  bone2Name=head
  swing1LimitDegrees=45
  swing1Motion=Free
  swing2LimitDegrees=45
  swing2Motion=Free
  twistLimitDegrees=45
  twistMotion=Free
 }
 {
  bone1Name=dyn_head_leaf_039_MSTR
  bone2Name=head
  swing1LimitDegrees=45
  swing1Motion=Free
  swing2LimitDegrees=45
  swing2Motion=Free
  twistLimitDegrees=45
  twistMotion=Free
 }
 {
  bone1Name=dyn_head_leaf_026_MSTR
  bone2Name=head
  swing1LimitDegrees=45
  swing1Motion=Free
  swing2LimitDegrees=45
  swing2Motion=Free
  twistLimitDegrees=45
  twistMotion=Free
 }
 {
  bone1Name=dyn_head_leaf_02_MSTR
  bone2Name=head
  swing1LimitDegrees=45
  swing1Motion=Free
  swing2LimitDegrees=45
  swing2Motion=Free
  twistLimitDegrees=45
  twistMotion=Free
 }
 {
  bone1Name=dyn_head_leaf_032_MSTR
  bone2Name=head
  swing1LimitDegrees=45
  swing1Motion=Free
  swing2LimitDegrees=45
  swing2Motion=Free
  twistLimitDegrees=45
  twistMotion=Free
 }
 {
  bone1Name=dyn_head_leaf_016_MSTR
  bone2Name=head
  swing1LimitDegrees=45
  swing1Motion=Free
  swing2LimitDegrees=45
  swing2Motion=Free
  twistLimitDegrees=45
  twistMotion=Free
 }
 {
  bone1Name=dyn_head_leaf_037_MSTR
  bone2Name=head
  swing1LimitDegrees=45
  swing1Motion=Free
  swing2LimitDegrees=45
  swing2Motion=Free
  twistLimitDegrees=45
  twistMotion=Free
 }
 {
  bone1Name=dyn_head_leaf_017_MSTR
  bone2Name=head
  swing1LimitDegrees=45
  swing1Motion=Free
  swing2LimitDegrees=45
  swing2Motion=Free
  twistLimitDegrees=45
  twistMotion=Free
 }
 {
  bone1Name=dyn_head_leaf_029_MSTR
  bone2Name=head
  swing1LimitDegrees=45
  swing1Motion=Free
  swing2LimitDegrees=45
  swing2Motion=Free
  twistLimitDegrees=45
  twistMotion=Free
 }
 {
  bone1Name=dyn_head_leaf_08_MSTR
  bone2Name=head
  swing1LimitDegrees=45
  swing1Motion=Free
  swing2LimitDegrees=45
  swing2Motion=Free
  twistLimitDegrees=45
  twistMotion=Free
 }
 {
  bone1Name=dyn_head_leaf_027_MSTR
  bone2Name=head
  swing1LimitDegrees=45
  swing1Motion=Free
  swing2LimitDegrees=45
  swing2Motion=Free
  twistLimitDegrees=45
  twistMotion=Free
 }
 {
  bone1Name=dyn_head_leaf_028_MSTR
  bone2Name=head
  swing1LimitDegrees=45
  swing1Motion=Free
  swing2LimitDegrees=45
  swing2Motion=Free
  twistLimitDegrees=45
  twistMotion=Free
 }
 {
  bone1Name=dyn_head_leaf_035_MSTR
  bone2Name=head
  swing1LimitDegrees=45
  swing1Motion=Free
  swing2LimitDegrees=45
  swing2Motion=Free
  twistLimitDegrees=45
  twistMotion=Free
 }
 {
  bone1Name=dyn_head_leaf_012_MSTR
  bone2Name=head
  swing1LimitDegrees=45
  swing1Motion=Free
  swing2LimitDegrees=45
  swing2Motion=Free
  twistLimitDegrees=45
  twistMotion=Free
 }
 {
  bone1Name=dyn_head_leaf_07_MSTR
  bone2Name=head
  swing1LimitDegrees=45
  swing1Motion=Free
  swing2LimitDegrees=45
  swing2Motion=Free
  twistLimitDegrees=45
  twistMotion=Free
 }
 {
  bone1Name=dyn_head_leaf_05_MSTR
  bone2Name=head
  swing1LimitDegrees=45
  swing1Motion=Free
  swing2LimitDegrees=45
  swing2Motion=Free
  twistLimitDegrees=45
  twistMotion=Free
 }
 {
  bone1Name=dyn_head_leaf_03_MSTR
  bone2Name=head
  swing1LimitDegrees=45
  swing1Motion=Free
  swing2LimitDegrees=45
  swing2Motion=Free
  twistLimitDegrees=45
  twistMotion=Free
 }
 {
  bone1Name=dyn_head_leaf_023_MSTR
  bone2Name=head
  swing1LimitDegrees=45
  swing1Motion=Free
  swing2LimitDegrees=45
  swing2Motion=Free
  twistLimitDegrees=45
  twistMotion=Free
 }
 {
  bone1Name=dyn_head_leaf_015_MSTR
  bone2Name=head
  swing1LimitDegrees=45
  swing1Motion=Free
  swing2LimitDegrees=45
  swing2Motion=Free
  twistLimitDegrees=45
  twistMotion=Free
 }
 {
  bone1Name=dyn_head_leaf_014_MSTR
  bone2Name=head
  swing1LimitDegrees=45
  swing1Motion=Free
  swing2LimitDegrees=45
  swing2Motion=Free
  twistLimitDegrees=45
  twistMotion=Free
 }
 {
  bone1Name=dyn_head_leaf_06_MSTR
  bone2Name=head
  swing1LimitDegrees=45
  swing1Motion=Free
  swing2LimitDegrees=45
  swing2Motion=Free
  twistLimitDegrees=45
  twistMotion=Free
 }
 {
  bone1Name=dyn_head_leaf_036_MSTR
  bone2Name=head
  swing1LimitDegrees=45
  swing1Motion=Free
  swing2LimitDegrees=45
  swing2Motion=Free
  twistLimitDegrees=45
  twistMotion=Free
 }
 {
  bone1Name=dyn_head_leaf_034_MSTR
  bone2Name=head
  swing1LimitDegrees=45
  swing1Motion=Free
  swing2LimitDegrees=45
  swing2Motion=Free
  twistLimitDegrees=45
  twistMotion=Free
 }
 {
  bone1Name=dyn_head_leaf_033_MSTR
  bone2Name=head
  swing1LimitDegrees=45
  swing1Motion=Free
  swing2LimitDegrees=45
  swing2Motion=Free
  twistLimitDegrees=45
  twistMotion=Free
 }
 {
  bone1Name=dyn_head_leaf_09_MSTR
  bone2Name=head
  swing1LimitDegrees=45
  swing1Motion=Free
  swing2LimitDegrees=45
  swing2Motion=Free
  twistLimitDegrees=45
  twistMotion=Free
 }
 {
  bone1Name=dyn_head_leaf_01_MSTR
  bone2Name=head
  swing1LimitDegrees=45
  swing1Motion=Free
  swing2LimitDegrees=45
  swing2Motion=Free
  twistLimitDegrees=45
  twistMotion=Free
 }
 {
  bone1Name=dyn_head_leaf_020_MSTR
  bone2Name=head
  swing1LimitDegrees=45
  swing1Motion=Free
  swing2LimitDegrees=45
  swing2Motion=Free
  twistLimitDegrees=45
  twistMotion=Free
 }
 {
  bone1Name=dyn_head_leaf_019_MSTR
  bone2Name=head
  swing1LimitDegrees=45
  swing1Motion=Free
  swing2LimitDegrees=45
  swing2Motion=Free
  twistLimitDegrees=45
  twistMotion=Free
 }
 {
  bone1Name=dyn_head_leaf_018_MSTR
  bone2Name=head
  swing1LimitDegrees=45
  swing1Motion=Free
  swing2LimitDegrees=45
  swing2Motion=Free
  twistLimitDegrees=45
  twistMotion=Free
 }
 {
  bone1Name=dyn_head_leaf_013_MSTR
  bone2Name=head
  swing1LimitDegrees=45
  swing1Motion=Free
  swing2LimitDegrees=45
  swing2Motion=Free
  twistLimitDegrees=45
  twistMotion=Free
 }
 {
  bone1Name=dyn_head_leaf_041_MSTR
  bone2Name=head
  swing1LimitDegrees=45
  swing1Motion=Free
  swing2LimitDegrees=45
  swing2Motion=Free
  twistLimitDegrees=45
  twistMotion=Free
 }
 {
  bone1Name=dyn_head_leaf_042_MSTR
  bone2Name=head
  swing1LimitDegrees=45
  swing1Motion=Free
  swing2LimitDegrees=45
  swing2Motion=Free
  twistLimitDegrees=45
  twistMotion=Free
 }
 {
  bone1Name=dyn_head_leaf_040_MSTR
  bone2Name=head
  swing1LimitDegrees=45
  swing1Motion=Free
  swing2LimitDegrees=45
  swing2Motion=Free
  twistLimitDegrees=45
  twistMotion=Free
 }
 {
  bone1Name=dyn_head_leaf_030_MSTR
  bone2Name=head
  swing1LimitDegrees=45
  swing1Motion=Free
  swing2LimitDegrees=45
  swing2Motion=Free
  twistLimitDegrees=45
  twistMotion=Free
 }
]
